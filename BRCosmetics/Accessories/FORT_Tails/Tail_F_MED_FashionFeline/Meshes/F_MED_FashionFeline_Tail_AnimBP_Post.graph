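# Post-process Animation Blueprint for the FashionFeline tail accessory that adjusts the tail pose with additive bone rotations.
# The AnimGraph chains Modify Bone nodes in component space over the base and first three tail joints, stacking repeated passes.
# Each uses bool-driven alpha with eased blending, much slower on the final pair.
# Each update sets isCrouchingStatic when crouching without 2D movement, though the shown AnimGraph has no visible read of it.
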AnimBlueprint /BRCosmetics/Accessories/FORT_Tails/Tail_F_MED_FashionFeline/Meshes/F_MED_FashionFeline_Tail_AnimBP_Post
target_skeleton=/Game/Accessories/FORT_Tails/Common/Fortnite_Base_Tail_Skeleton
graphs [
 AnimGraph {nodes [
   n1 = "Transform (Modify) Bone - Bone: C_Tail_A_Base_Jnt"!(ComponentPose=Local To Component(LocalPose=Input Pose()), bAlphaBoolEnabled:bool=True, Rotation="0,0,60") -> (Pose) {
    alphaInputType=Bool
    rotationMode=BMM_Additive
    alphaBoolBlend {blendInTime=0.5, blendOption=CubicInOut, blendOutTime=0.5}
    boneToModify {boneName=C_Tail_A_Base_Jnt}
    rotation {roll=60}
   }
   n2 = "Transform (Modify) Bone - Bone: C_Tail_A_1_Jnt"!(ComponentPose=@n1, bAlphaBoolEnabled:bool=True, Rotation="-20,0,50") -> (Pose) {
    alphaInputType=Bool
    rotationMode=BMM_Additive
    alphaBoolBlend {blendInTime=0.5, blendOption=CubicInOut, blendOutTime=0.5}
    boneToModify {boneName=C_Tail_A_1_Jnt}
    rotation {pitch=-20, roll=50}
   }
   n3 = "Transform (Modify) Bone - Bone: C_Tail_A_2_Jnt"!(ComponentPose=@n2, bAlphaBoolEnabled:bool=True, Rotation="0,0,20") -> (Pose) {
    alphaInputType=Bool
    rotationMode=BMM_Additive
    alphaBoolBlend {blendInTime=0.5, blendOption=CubicInOut, blendOutTime=0.5}
    boneToModify {boneName=C_Tail_A_2_Jnt}
    rotation {roll=20}
   }
   n4 = "Transform (Modify) Bone - Bone: C_Tail_A_3_Jnt"!(ComponentPose=@n3, bAlphaBoolEnabled:bool=True, Rotation="0,0,20") -> (Pose) {
    alphaInputType=Bool
    rotationMode=BMM_Additive
    alphaBoolBlend {blendInTime=0.5, blendOption=CubicInOut, blendOutTime=0.5}
    boneToModify {boneName=C_Tail_A_3_Jnt}
    rotation {roll=20}
   }
   n5 = "Transform (Modify) Bone - Bone: C_Tail_A_Base_Jnt"(ComponentPose=@n4, bAlphaBoolEnabled=True, Rotation="0,0,60") {
    alphaInputType=Bool
    rotationMode=BMM_Additive
    alphaBoolBlend {blendInTime=0.5, blendOption=CubicInOut, blendOutTime=0.5}
    boneToModify {boneName=C_Tail_A_Base_Jnt}
    rotation {roll=60}
   }
   n6 = "Transform (Modify) Bone - Bone: C_Tail_A_1_Jnt"(ComponentPose=@n5, bAlphaBoolEnabled=True, Rotation="0,0,50") {
    alphaInputType=Bool
    rotationMode=BMM_Additive
    alphaBoolBlend {blendInTime=0.5, blendOption=CubicInOut, blendOutTime=0.5}
    boneToModify {boneName=C_Tail_A_1_Jnt}
    rotation {roll=50}
   }
   n7 = "Transform (Modify) Bone - Bone: C_Tail_A_2_Jnt"(ComponentPose=@n6, bAlphaBoolEnabled=True, Rotation="0,0,30") {
    alphaInputType=Bool
    rotationMode=BMM_Additive
    alphaBoolBlend {blendInTime=0.5, blendOption=CubicInOut, blendOutTime=0.5}
    boneToModify {boneName=C_Tail_A_2_Jnt}
    rotation {roll=30}
   }
   n8 = "Transform (Modify) Bone - Bone: C_Tail_A_3_Jnt"(ComponentPose=@n7, bAlphaBoolEnabled=True, Rotation="0,0,30") {
    alphaInputType=Bool
    rotationMode=BMM_Additive
    alphaBoolBlend {blendInTime=0.5, blendOption=CubicInOut, blendOutTime=0.5}
    boneToModify {boneName=C_Tail_A_3_Jnt}
    rotation {roll=30}
   }
   n9 = "Transform (Modify) Bone - Bone: C_Tail_A_Base_Jnt"(ComponentPose=@n8, bAlphaBoolEnabled=True, Rotation="0,0,15") {
    alpha=0
    alphaInputType=Bool
    rotationMode=BMM_Additive
    alphaBoolBlend {blendInTime=5, blendOption=Cubic, blendOutTime=10}
    boneToModify {boneName=C_Tail_A_Base_Jnt}
    rotation {roll=15}
   }
   n10 = "Transform (Modify) Bone - Bone: C_Tail_A_1_Jnt"(ComponentPose=@n9, bAlphaBoolEnabled=True, Rotation="0,0,5") {
    alpha=0
    alphaInputType=Bool
    rotationMode=BMM_Additive
    alphaBoolBlend {blendInTime=5, blendOption=Cubic, blendOutTime=10}
    boneToModify {boneName=C_Tail_A_1_Jnt}
    rotation {roll=5}
   }
   Output Pose(Result=Component To Local(ComponentPose=@n10))
  ]}
 EventGraph {nodes [
   Event BlueprintUpdateAnimation()
   n1 = Get bIsMoving2D()
   > Set isCrouchingStatic(isCrouchingStatic:bool=AND Boolean(A=NOT Boolean(A=@n1), B=Get bIsCrouching())) -> (Output_Get:bool)
  ]}
]
variables [isCrouchingStatic {type=boolean}]
